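# Animation Blueprint for the SandScreen vehicle body, procedurally posing its skeleton from the mesh reference pose with no animation assets.
# A Modify Bone chain applies axle, wheel position, steering, spin and steering-wheel variables, none declared or computed here.
# When bHasTireMod is true, additive offsets shift the axle pivots and widen the wheel-well spreader bones.
# A Vehicle Bouncy Chassis Controller finishes the pose before output.
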AnimBlueprint /VehicleCosmetics/Bodies/SandScreen/Animation/SK_Body_SandScreen_AnimBP
target_skeleton=/VehicleCosmetics/Bodies/SandScreen/Meshes/SK_Body_SandScreen_Skeleton
graphs [
 AnimGraph {nodes [
   n1 = "Transform (Modify) Bone - Bone: front_differential"(
    ComponentPose=Mesh Space Ref Pose!() -> (ComponentPose:Component Space Pose Link Structure) {is=title AnimGraphNode_MeshRefPose}
    Alpha=1
    Translation=Get AxlePositionF() -> (AxlePositionF:Vector)
    Rotation=Get AxleRotationF() -> (AxleRotationF:Rotator)
    Scale="1,1,1"
   ) {
    rotationMode=BMM_Replace
    rotationSpace=BCS_BoneSpace
    translationMode=BMM_Replace
    translationSpace=BCS_BoneSpace
    alphaBoolBlend {blendOption=Linear}
    boneToModify {boneName=front_differential}
   }
   n2 = "Transform (Modify) Bone - Bone: rear_differential"(
    ComponentPose=@n1
    Alpha=1
    Translation=Get AxlePositionB() -> (AxlePositionB:Vector)
    Rotation=Get AxleRotationB() -> (AxleRotationB:Rotator)
    Scale="1,1,1"
   ) {
    rotationMode=BMM_Replace
    rotationSpace=BCS_BoneSpace
    translationMode=BMM_Replace
    translationSpace=BCS_BoneSpace
    alphaBoolBlend {blendOption=Linear}
    boneToModify {boneName=rear_differential}
   }
   n3 = "Transform (Modify) Bone - Bone: axle_pivot_fr_l"(ComponentPose=@n2, Alpha=1, Translation=Get WheelPositionFL() -> (WheelPositionFL:Vector), Rotation="0,0,0", Scale="1,1,1") {
    rotationMode=BMM_Replace
    rotationSpace=BCS_BoneSpace
    translationMode=BMM_Replace
    translationSpace=BCS_BoneSpace
    alphaBoolBlend {blendOption=Linear}
    boneToModify {boneName=axle_pivot_fr_l}
   }
   n4 = "Transform (Modify) Bone - Bone: axle_pivot_fr_r"(ComponentPose=@n3, Alpha=1, Translation=Get WheelPositionFR() -> (WheelPositionFR:Vector), Rotation="0,0,0", Scale="1,1,1") {
    rotationMode=BMM_Replace
    rotationSpace=BCS_BoneSpace
    translationMode=BMM_Replace
    translationSpace=BCS_BoneSpace
    alphaBoolBlend {blendOption=Linear}
    boneToModify {boneName=axle_pivot_fr_r}
   }
   n5 = "Transform (Modify) Bone - Bone: axle_pivot_bk_r"(ComponentPose=@n4, Alpha=1, Translation=Get WheelPositionBR() -> (WheelPositionBR:Vector), Rotation="0,0,0", Scale="1,1,1") {
    rotationMode=BMM_Replace
    rotationSpace=BCS_BoneSpace
    translationMode=BMM_Replace
    translationSpace=BCS_BoneSpace
    alphaBoolBlend {blendOption=Linear}
    boneToModify {boneName=axle_pivot_bk_r}
   }
   n6 = "Transform (Modify) Bone - Bone: axle_pivot_bk_l"(ComponentPose=@n5, Alpha=1, Translation=Get WheelPositionBL() -> (WheelPositionBL:Vector), Rotation="0,0,0", Scale="1,1,1") {
    rotationMode=BMM_Replace
    rotationSpace=BCS_BoneSpace
    translationMode=BMM_Replace
    translationSpace=BCS_BoneSpace
    alphaBoolBlend {blendOption=Linear}
    boneToModify {boneName=axle_pivot_bk_l}
   }
   n7 = "Transform (Modify) Bone - Bone: wheel_steering_fr_r"(ComponentPose=@n6, Alpha=1, Translation="0,0,0", Rotation=Get WheelRotationFR() -> (WheelRotationFR:Rotator), Scale="1,1,1") {
    rotationMode=BMM_Replace
    rotationSpace=BCS_BoneSpace
    translationSpace=BCS_BoneSpace
    alphaBoolBlend {blendOption=Linear}
    boneToModify {boneName=wheel_steering_fr_r}
   }
   n8 = "Transform (Modify) Bone - Bone: wheel_steering_fr_l"(ComponentPose=@n7, Alpha=1, Translation="0,0,0", Rotation=Get WheelRotationFL() -> (WheelRotationFL:Rotator), Scale="1,1,1") {
    rotationMode=BMM_Replace
    rotationSpace=BCS_BoneSpace
    translationSpace=BCS_BoneSpace
    alphaBoolBlend {blendOption=Linear}
    boneToModify {boneName=wheel_steering_fr_l}
   }
   n9 = "Transform (Modify) Bone - Bone: wheel_disc_fr_r"(ComponentPose=@n8, Alpha=1, Translation="0,0,0", Rotation=Get WheelSpinFR() -> (WheelSpinFR:Rotator), Scale="1,1,1") {rotationMode=BMM_Replace, rotationSpace=BCS_BoneSpace, alphaBoolBlend {blendOption=Linear}, boneToModify {boneName=wheel_disc_fr_r}}
   n10 = "Transform (Modify) Bone - Bone: wheel_disc_fr_l"(ComponentPose=@n9, Alpha=1, Translation="0,0,0", Rotation=Get WheelSpinFL() -> (WheelSpinFL:Rotator), Scale="1,1,1") {rotationMode=BMM_Replace, rotationSpace=BCS_BoneSpace, alphaBoolBlend {blendOption=Linear}, boneToModify {boneName=wheel_disc_fr_l}}
   n11 = "Transform (Modify) Bone - Bone: wheel_disc_bk_r"(ComponentPose=@n10, Alpha=1, Translation="0,0,0", Rotation=Get WheelSpinBR() -> (WheelSpinBR:Rotator), Scale="1,1,1") {rotationMode=BMM_Replace, rotationSpace=BCS_BoneSpace, alphaBoolBlend {blendOption=Linear}, boneToModify {boneName=wheel_disc_bk_r}}
   n12 = "Transform (Modify) Bone - Bone: wheel_disc_bk_l"(ComponentPose=@n11, Alpha=1, Translation="0,0,0", Rotation=Get WheelSpinBL() -> (WheelSpinBL:Rotator), Scale="1,1,1") {rotationMode=BMM_Replace, rotationSpace=BCS_BoneSpace, alphaBoolBlend {blendOption=Linear}, boneToModify {boneName=wheel_disc_bk_l}}
   n13 = "Transform (Modify) Bone - Bone: steering_wheel"(ComponentPose=@n12, Alpha=1, Translation="0,0,0", Rotation=Get SteeringRotation() -> (SteeringRotation:Rotator), Scale="1,1,1") {rotationMode=BMM_Replace, rotationSpace=BCS_BoneSpace, alphaBoolBlend {blendOption=Linear}, boneToModify {boneName=steering_wheel}}
   n14 = Get bHasTireMod() -> (bHasTireMod:bool)
   n15 = "Transform (Modify) Bone - Bone: axle_pivot_bk_l"!(ComponentPose=@n13, bAlphaBoolEnabled:bool=@n14, Translation="0,-20,-4", Rotation="0,0,0", Scale="1,1,1") -> (Pose) {
    alphaInputType=Bool
    rotationSpace=BCS_BoneSpace
    scaleSpace=BCS_BoneSpace
    translationMode=BMM_Additive
    translationSpace=BCS_BoneSpace
    alphaBoolBlend {blendOption=Linear}
    boneToModify {boneName=axle_pivot_bk_l}
    translation {y=-20, z=-4}
   }
   n16 = "Transform (Modify) Bone - Bone: axle_pivot_bk_r"!(ComponentPose=@n15, bAlphaBoolEnabled:bool=@n14, Translation="0,20,-4", Rotation="0,0,0", Scale="1,1,1") -> (Pose) {
    alphaInputType=Bool
    rotationSpace=BCS_BoneSpace
    scaleSpace=BCS_BoneSpace
    translationMode=BMM_Additive
    translationSpace=BCS_BoneSpace
    alphaBoolBlend {blendOption=Linear}
    boneToModify {boneName=axle_pivot_bk_r}
    translation {y=20, z=-4}
   }
   n17 = "Transform (Modify) Bone - Bone: axle_pivot_fr_l"!(ComponentPose=@n16, bAlphaBoolEnabled:bool=@n14, Translation="0,-18,-4", Rotation="0,0,0", Scale="1,1,1") -> (Pose) {
    alphaInputType=Bool
    rotationSpace=BCS_BoneSpace
    translationMode=BMM_Additive
    translationSpace=BCS_BoneSpace
    alphaBoolBlend {blendOption=Linear}
    boneToModify {boneName=axle_pivot_fr_l}
    translation {y=-18, z=-4}
   }
   n18 = "Transform (Modify) Bone - Bone: axle_pivot_fr_r"!(ComponentPose=@n17, bAlphaBoolEnabled:bool=@n14, Translation="0,18,-4", Rotation="0,0,0", Scale="1,1,1") -> (Pose) {
    alphaInputType=Bool
    rotationSpace=BCS_BoneSpace
    translationMode=BMM_Additive
    translationSpace=BCS_BoneSpace
    alphaBoolBlend {blendOption=Linear}
    boneToModify {boneName=axle_pivot_fr_r}
    translation {y=18, z=-4}
   }
   n19 = "Transform (Modify) Bone - Bone: wheel_well_spreader_bk_l"!(ComponentPose=@n18, bAlphaBoolEnabled:bool=@n14, Translation="3,0,0", Rotation="0,0,0", Scale="1.25,1,1") -> (Pose) {
    alphaInputType=Bool
    rotationSpace=BCS_BoneSpace
    scaleMode=BMM_Additive
    scaleSpace=BCS_BoneSpace
    translationMode=BMM_Additive
    translationSpace=BCS_BoneSpace
    alphaBoolBlend {blendOption=Linear}
    alphaScaleBiasClamp {bInterpResult=true, interpSpeedDecreasing=5, interpSpeedIncreasing=5}
    boneToModify {boneName=wheel_well_spreader_bk_l}
    scale {x=1.25}
    translation {x=3}
   }
   n20 = "Transform (Modify) Bone - Bone: wheel_well_spreader_bk_r"!(ComponentPose=@n19, bAlphaBoolEnabled:bool=@n14, Translation="3,0,0", Rotation="0,0,0", Scale="1.25,1,1") -> (Pose) {
    alphaInputType=Bool
    rotationSpace=BCS_BoneSpace
    scaleMode=BMM_Additive
    scaleSpace=BCS_BoneSpace
    translationMode=BMM_Additive
    translationSpace=BCS_BoneSpace
    alphaBoolBlend {blendOption=Linear}
    alphaScaleBiasClamp {bInterpResult=true, interpSpeedDecreasing=5, interpSpeedIncreasing=5}
    boneToModify {boneName=wheel_well_spreader_bk_r}
    scale {x=1.25}
    translation {x=3}
   }
   n21 = "Transform (Modify) Bone - Bone: wheel_well_spreader_fr_l"!(ComponentPose=@n20, bAlphaBoolEnabled:bool=@n14, Translation="0,0,0", Rotation="0,0,0", Scale="1.25,1,1") -> (Pose) {
    alphaInputType=Bool
    rotationSpace=BCS_BoneSpace
    scaleMode=BMM_Additive
    scaleSpace=BCS_BoneSpace
    translationMode=BMM_Additive
    translationSpace=BCS_BoneSpace
    alphaBoolBlend {blendOption=Linear}
    alphaScaleBiasClamp {bInterpResult=true, interpSpeedDecreasing=5, interpSpeedIncreasing=5}
    boneToModify {boneName=wheel_well_spreader_fr_l}
    scale {x=1.25}
   }
   n22 = "Transform (Modify) Bone - Bone: wheel_well_spreader_fr_r"!(ComponentPose=@n21, bAlphaBoolEnabled:bool=@n14, Translation="0,0,0", Rotation="0,0,0", Scale="1.25,1,1") -> (Pose) {
    alphaInputType=Bool
    rotationSpace=BCS_BoneSpace
    scaleMode=BMM_Additive
    scaleSpace=BCS_BoneSpace
    translationMode=BMM_Additive
    translationSpace=BCS_BoneSpace
    alphaBoolBlend {blendOption=Linear}
    alphaScaleBiasClamp {bInterpResult=true, interpSpeedDecreasing=5, interpSpeedIncreasing=5}
    boneToModify {boneName=wheel_well_spreader_fr_r}
    scale {x=1.25}
   }
   Output Pose(Result=Component To Local(ComponentPose=Vehicle Bouncy Chassis Controller!(
      ComponentPose:Component Space Pose Link Structure=@n22
      Alpha:float=1
      BouncyChassis:Bouncy Chassis Vehicle State Structure=(PitchDegrees=0,RollDegrees=0,TranslationZ=0)
     ) -> (Pose:Component Space Pose Link Structure) {is=title AnimGraphNode_VehicleBouncyChassisController, alphaBoolBlend {blendOption=Linear}}))
  ]}
 EventGraph {nodes [
   TryGetPawnOwner()
   Event BlueprintUpdateAnimation()
  ]}
]
variables []
