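# Physics asset for the SandScreen vehicle body cosmetic, named as the BR variant.
# It defines a single physics body on the root bone, with no collision shapes listed.
# It has no constraints, so no joints link separate bodies.
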PhysicsAsset /VehicleCosmetics/Bodies/SandScreen/Meshes/Body_SandScreen_PhysicsBR
bodies {root []}
constraints []
